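# Sound cue for the final laugh in the FastCheetah emote, classed and routed as emote foley.
# A Random node picks one of five equally weighted laugh variations on each play.
# Emote attenuation, a property sheet and a concurrency set govern playback, with virtualization set to keep playing when silent.
# It also declares a send to the music audio-analysis submix.
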SoundCue /BRCosmetics/Sounds/Emotes/FastCheetah/EmoteFoley_FastCheetah_FinalLaugh
root=Random_0
nodes [
 {id=Random_0, children [WavePlayer_0, WavePlayer_15, WavePlayer_16, WavePlayer_17, WavePlayer_18], properties {weights [1,1,1,1,1]}}
 {id=WavePlayer_0, properties {soundWaveAssetPtr=&/IP_Emote_FastCheetah_FinalLaugh_05}}
 {id=WavePlayer_15, properties {soundWaveAssetPtr=&/IP_Emote_FastCheetah_FinalLaugh_04}}
 {id=WavePlayer_16, properties {soundWaveAssetPtr=&/IP_Emote_FastCheetah_FinalLaugh_03}}
 {id=WavePlayer_17, properties {soundWaveAssetPtr=&/IP_Emote_FastCheetah_FinalLaugh_02}}
 {id=WavePlayer_18, properties {soundWaveAssetPtr=&/IP_Emote_FastCheetah_FinalLaugh_01}}
]
properties {
 attenuationSettings=&Fort_Emotes_Athena_Att
 audioPropertiesSheet=/BRCosmetics/Sounds/Settings/PropertySheets/Emotes/APS_Cosmetic_Emote_Foley
 maxDistance=1600
 soundClassObject=/Game/Sounds/FortSoundClassesAndModes/Fort_Foley_Emote
 virtualizationMode=PlayWhenSilent
 volumeMultiplier=1
 concurrencyOverrides {maxCount=3}
 concurrencySet [/Game/Sounds/Concurrency_Rules/Athena_EmoteFoley_Conc]
 soundSubmixSends [{minSendLevel=0, soundSubmix=&Fortnite_AudioAnalysis_Music}]
}
